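; Autoplay video screen actor that plays looping media on its mesh while the local player is inside a Sphere trigger.
; Overlap and the user's video-playback setting create or clean up a MediaPlayer, swapping active and inactive materials.
; On open, it seeks to server world time modulo duration, keeping playback in sync across clients.
; On mobile, a timer sets native volume from view-target distance.
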(event EventBeginPlay
  ;[1 1 1] Cleanup at the beginning to make sure intermission texture is set
  (if (not (Variables|Default|GetDebugBypassCalendarEvent))
    (MCPContext|WatchCalendarEvent (Variables|Default|GetEventString)
      (:OnEventActive
        _)))
  (CallFunction|Cleanup))

(event OnComponentBeginOverlap(Sphere)
  (params (OverlappedComponent :type Primitive%20Component%20Object%20Reference) (OtherActor :type Actor%20Object%20Reference) (OtherComp :type Primitive%20Component%20Object%20Reference) (OtherBodyIndex :type Integer) (bFromSweep :type Boolean) (SweepResult :type Hit%20Result%20Structure%20%28by%20ref%29))
  (bind _returnvalue (Settings|GetFortGameUserSettings))
  ;[1 1 1] Enable and disable video based on distance Sphere trigger volume
  (if (Utilities|Equal(Object) (Game|GetPlayerPawn 0) OtherActor)
    ;[1 1 1] BeginPlay - Check calendar event active, if yes - Initialize media player, bind options delegate
    (if (not (Variables|Default|GetDebugBypassCalendarEvent))
      (MCPContext|WatchCalendarEvent (Variables|Default|GetEventString)
        (:OnEventActive
          _)))
    (Utilities|FlowControl|Delay 1.0)
    (EventDispatchers|BindEventtoOnAllowVideoPlaybackChanged _returnvalue (delegate Custom|OnAllowPlaybackChanged))
    (bind _returnvalue_2 (Class|FortGameUserSettings|IsVideoPlaybackAllowed _returnvalue))
    (if _returnvalue_2
      (CallFunction|Initialize))))

(event EventEndPlay
  (params (EndPlayReason :type EEndPlayReason%20Enum :default "Destroyed"))
  ;[1 1 1] EndPlay - Tidy media player, unbind delegate
  (EventDispatchers|UnbindEventfromOnAllowVideoPlaybackChanged (Settings|GetFortGameUserSettings) (delegate Custom|OnAllowPlaybackChanged))
  (CallFunction|Cleanup))

(event OnComponentEndOverlap(Sphere)
  (params (OverlappedComponent :type Primitive%20Component%20Object%20Reference) (OtherActor :type Actor%20Object%20Reference) (OtherComp :type Primitive%20Component%20Object%20Reference) (OtherBodyIndex :type Integer))
  ;[1 1 1] Enable and disable video based on distance Sphere trigger volume
  (if (Utilities|Equal(Object) (Game|GetPlayerPawn 0) OtherActor)
    ;[1 1 1] EndPlay - Tidy media player, unbind delegate
    (EventDispatchers|UnbindEventfromOnAllowVideoPlaybackChanged (Settings|GetFortGameUserSettings) (delegate Custom|OnAllowPlaybackChanged))
    (CallFunction|Cleanup)))

(event Custom|OnAllowPlaybackChanged
  (params (bAllowed :type Boolean))
  ;[1 1 1] Options delegate - update our state to match if needed
  (if (Math|Boolean|NotEqual(Boolean) bAllowed (Utilities|IsValid (Variables|Default|GetMediaPlayer) :!class /Script/Engine.KismetSystemLibrary))
    (if bAllowed
      (CallFunction|Initialize)
      (else
        (CallFunction|Cleanup)))))

(event Custom|Initialize
  ;[1 1 1] General initialization
  (Networking|CanExecuteCosmeticEvents
    (:True
      (CallFunction|CreateAndStartMediaPlayer)
      ;[1 1 1] Only kick this off if on a mobile platform
      (if (Fortnite|IsMobilePlatform)
        (Utilities|Time|SetTimerbyEvent (EventDispatchers|CreateEvent :!event Custom|Evaluate%20Volume%20Attenuation) (Variables|Default|GetTickRate) true)
        (CallFunction|EvaluateVolumeAttenuation)))))

(event Custom|Create%20and%20Start%20Media%20Player
  ;[1 1 1] Create Media Player
  (Rendering|Material|SetMaterial (Variables|StaticMeshActor|GetStaticMeshComponent) 0 (Variables|Default|GetActivePlayerMaterial))
  (if (not (Utilities|IsValid (Variables|Default|GetMediaPlayer) :!class /Script/Engine.KismetSystemLibrary))
    (bind _returnvalue (Game|ConstructObjectfromClass "/Script/MediaAssets.MediaPlayer" self))
    (Variables|Default|SetMediaPlayer _returnvalue)
    (Media|MediaPlayer|SetLooping (Variables|Default|GetMediaPlayer) true))
  (Media|MediaSoundComponent|SetMediaPlayer (Variables|Default|GetMediaSound) (Variables|Default|GetMediaPlayer))
  (Media|MediaTexture|SetMediaPlayer (Variables|Default|GetMedia%20Texture) (Variables|Default|GetMediaPlayer) :!class /Script/MediaAssets.MediaTexture)
  (Media|MediaPlayer|AssignOnMediaOpened (Variables|Default|GetMediaPlayer) (EventDispatchers|CreateEvent :!event Custom|On%20Media%20Opened))
  (Media|MediaPlayer|OpenSource (Variables|Default|GetMediaPlayer) (Variables|Default|GetPlatformSource)))

(event Custom|On%20Media%20Opened
  (params (OpenedUrl :type String))
  ;[1 1 1] Play Media after opening
  (Variables|Default|SetMediaDuration (Math|Timespan|GetTotalSeconds (Media|MediaPlayer|GetDuration (Variables|Default|GetMediaPlayer))))
  (Variables|Default|SetInitialSeek (Math|Float|%25%28Float%29 (GameState|GetServerWorldTimeSeconds (Game|GetGameState)) (Variables|Default|GetMediaDuration)))
  (Media|MediaPlayer|Play (Variables|Default|GetMediaPlayer) :!class /Script/MediaAssets.MediaPlayer)
  (Media|MediaPlayer|SetNativeVolume (Variables|Default|GetMediaPlayer) 0.0)
  (Media|MediaPlayer|Seek (Variables|Default|GetMediaPlayer) (Math|Timespan|FromSeconds (Variables|Default|GetInitialSeek))))

(event Custom|Cleanup
  ;[1 1 1] Cleanup
  (Rendering|Material|SetMaterial (Variables|StaticMeshActor|GetStaticMeshComponent) 0 (Variables|Default|GetInactivePlayerMaterial))
  (if (Utilities|IsValid (Variables|Default|GetMediaPlayer) :!class /Script/Engine.KismetSystemLibrary)
    (Media|MediaPlayer|Close (Variables|Default|GetMediaPlayer) :!class /Script/MediaAssets.MediaPlayer)
    (Variables|Default|SetMediaPlayer 0)
    (Media|MediaSoundComponent|SetMediaPlayer (Variables|Default|GetMediaSound) 0)
    (Media|MediaTexture|SetMediaPlayer (Variables|Default|GetMedia%20Texture) 0 :!class /Script/MediaAssets.MediaTexture)))

(event Custom|Evaluate%20Volume%20Attenuation
  ;[1 1 1] Mobile platform volume attenuation
  (bind _returnvalue (Audio|AreAnyListenersWithinRange (Transformation|GetActorLocation) (Variables|Default|GetMaxAudibleDistance)))
  (if _returnvalue
    (Variables|Default|GetLocalFortPC
      (:then
        _)
      (:else
        (CallFunction|CacheLocalFortPC)
        (bind _localfortpc (Variables|Default|GetLocalFortPC)
          (:then
            _)
          (:else))))
    (Utilities|IsValid (Pawn|GetViewTarget _localfortpc)
      (:Is%20Valid
        (bind _distance (Class|MissionBlueprintFunctionLibrary|DistanceBetweenTwoVectors (Transformation|GetActorLocation (Pawn|GetViewTarget _localfortpc)) (Transformation|GetActorLocation)))
        (bind _output_get (Variables|Default|SetMobileVolumeTarget (Math|Float|MapRangeClamped _distance (Variables|Default|GetMinAudibleDistance) (Variables|Default|GetMaxAudibleDistance) 1.0)))
        (bind _returnvalue_2 (Media|MediaPlayer|SetNativeVolume (Variables|Default|GetMediaPlayer) _output_get))
        (Development|PrintString (Utilities|String|Append "Set Native Volume (" (Utilities|String|ToString(Boolean) _returnvalue_2) "): " (Utilities|String|ToString(Float) _output_get)) (Variables|Default|GetbDebugVolumeAttenuation) (Variables|Default|GetbDebugVolumeAttenuation))))
    (else
      (Media|MediaPlayer|SetNativeVolume (Variables|Default|GetMediaPlayer) 0.0))))

(fn CacheLocalFortPC
  (bind _returnvalue (Fortnite|GetLocalFortPlayerControllers true))
  (if (> (Utilities|Array|Length _returnvalue) 0)
    (Variables|Default|SetLocalFortPC (Utilities|Array|Get(acopy) _returnvalue))))

(fn ConstructionScript)
